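# Physics asset for the F_MED_Relish backpack cosmetic, defining collision and joints for a dangling hammer and a paper piece.
# The hammer is a capsule constrained to a small sphere at its base bone with tight angular limits.
# The paper is three thin boxes chained from its own sphere base, with swing2 limits much wider than swing1 and twist, so it flexes mostly about one axis.
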PhysicsAsset /BRCosmetics/Accessories/FORT_Backpacks/Backpack_F_MED_Relish/Meshes/F_MED_Relish_Pack_Physics
bodies {
 dyn_hammer_base [{shapeType=Sphere, radius=0.5, shapeName=dyn_hammer_base_sphere}]
 dyn_hammer_parent [{shapeType=Capsule, length=4.7488, radius=0.9984, shapeName=dyn_hammer_parent_capsule, center=<3.0393 -0.0001 0>, rotation=<rot 90 0 0>}]
 dyn_paper_01 [{
   shapeType=Box
   extentX=6.8361
   extentY=23.2521
   extentZ=1.1579
   shapeName=dyn_paper_01_box
   center=<3.4409 0.1686 0.9817>
   rotation=<rot 17.8466 4.2961 0>
  }]
 dyn_paper_02 [{shapeType=Box, extentX=8.7029, extentY=24.0498, extentZ=1.0048, shapeName=dyn_paper_02_box, center=<3.9214 0 0>}]
 dyn_paper_base [{shapeType=Sphere, radius=0.5, shapeName=dyn_paper_base_sphere}]
 dyn_paper_parent [{
   shapeType=Box
   extentX=4.01
   extentY=23.51
   extentZ=1.01
   shapeName=dyn_paper_parent_box
   center=<2.0137 0 0.0001>
   rotation=<rot 0 2.6089 0>
  }]
}
constraints [
 {bone1Name=dyn_hammer_parent, bone2Name=dyn_hammer_base, swing1LimitDegrees=2, swing2LimitDegrees=6, twistLimitDegrees=3}
 {bone1Name=dyn_paper_parent, bone2Name=dyn_paper_base, swing1LimitDegrees=2, swing2LimitDegrees=30, twistLimitDegrees=3}
 {bone1Name=dyn_paper_01, bone2Name=dyn_paper_parent, swing1LimitDegrees=2, swing2LimitDegrees=20, twistLimitDegrees=3}
 {bone1Name=dyn_paper_02, bone2Name=dyn_paper_01, swing1LimitDegrees=2, swing2LimitDegrees=20, twistLimitDegrees=3}
]
