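# Animation Blueprint for the F_MED_Relish backpack cosmetic, adding posture correction and secondary physics motion.
# Each update maps spine_05-to-spine_01 yaw on the owner's skeletal mesh to root roll and offset.
# The AnimGraph applies fixed paper and pack offsets and that root adjustment, then RigidBody simulation in root space, then shrinks the pack bone.
# Jump begin and end events set or clear, after short delays, an upward external force driving the simulation.
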AnimBlueprint /BRCosmetics/Accessories/FORT_Backpacks/Backpack_F_MED_Relish/Meshes/F_MED_Relish_Pack_AnimBP
target_skeleton=&/F_MED_Relish_Pack_Skeleton
graphs [
 AnimGraph {nodes [
   n1 = "Transform (Modify) Bone - Bone: dyn_paper_parent"(ComponentPose=Local To Component(LocalPose=Input Pose()), Alpha=1, Translation="0,0,0", Rotation="-7.660,-0.000,1.213", Scale="1,1,1") {
    rotationMode=BMM_Additive
    rotationSpace=BCS_BoneSpace
    translationMode=BMM_Additive
    translationSpace=BCS_BoneSpace
    alphaBoolBlend {blendOption=Linear}
    boneToModify {boneName=dyn_paper_parent}
    rotation=<rot -7.66 0 1.213>
   }
   n2 = "Transform (Modify) Bone - Bone: pack"(ComponentPose=@n1, Translation="0.000,0,-1.988", Rotation="0.000,0.000,-7.706") {
    rotationMode=BMM_Additive
    rotationSpace=BCS_BoneSpace
    translationMode=BMM_Additive
    translationSpace=BCS_BoneSpace
    alphaBoolBlend {blendOption=Linear}
    boneToModify {boneName=pack}
    rotation {roll=-7.706}
    translation {z=-1.988}
   }
   n3 = "Transform (Modify) Bone - Bone: root"(ComponentPose=@n2, Translation=Get backpack_translation(), Rotation=Get backpack_rotate()) {
    rotationMode=BMM_Additive
    rotationSpace=BCS_BoneSpace
    translationMode=BMM_Additive
    translationSpace=BCS_BoneSpace
    alphaBoolBlend {blendOption=Linear}
    boneToModify {boneName=root}
   }
   n4 = RigidBody(ComponentPose=@n3, ExternalForce=Get JumpVelocity() -> (JumpVelocity:Vector)) {
    bClampLinearTranslationLimitToRefPose=true
    bOverrideWorldGravity=true
    overridePhysicsAsset=&/F_MED_Relish_Pack_Physics
    simulationSpace=BaseBoneSpace
    alphaBoolBlend {blendOption=Linear}
    baseBoneRef {boneName=root}
    componentAppliedLinearAccClamp=<6000 6000 6000>
    componentLinearAccScale=<0.2 0.2 0.2>
    componentLinearVelScale=<0.2 0.2 0.2>
    overrideWorldGravity {z=-980}
   }
   n5 = "Transform (Modify) Bone - Bone: pack"(ComponentPose=@n4, Translation="0,0,0", Rotation="0,0,0", Scale="0.9,0.9,0.9") {
    rotationMode=BMM_Additive
    rotationSpace=BCS_BoneSpace
    scaleMode=BMM_Replace
    scaleSpace=BCS_BoneSpace
    translationMode=BMM_Additive
    translationSpace=BCS_BoneSpace
    alphaBoolBlend {blendOption=Linear}
    boneToModify {boneName=pack}
    scale=<0.9 0.9 0.9>
   }
   Output Pose(Result=Component To Local(ComponentPose=@n5)) {layerGroup=None}
  ]}
 EventGraph {nodes [
   Event BlueprintUpdateAnimation()
   n1 = GetOwnerSkeletalMeshComponent()
   n2 = BreakRotator(InRot=BreakTransform(InTransform=GetDeltaTransformFromRefPose(self=@n1, BoneName=spine_05, BaseName=spine_01)).Rotation)
   > Set backpack_rotate(backpack_rotate=MakeRotator(Roll=MapRangeClamped(Value=@n2.Yaw, InRangeA=0, InRangeB=-30, OutRangeB=20)))
   > Set backpack_translation(backpack_translation=MakeVector(Y=MapRangeClamped(Value=@n2.Yaw, InRangeA=0, InRangeB=-30, OutRangeB=-3)))
   Event OnBeginJumping()
   > RetriggerableDelay(Duration=0.25)
   > Set JumpVelocity(JumpVelocity:Vector="0, 0,5000") -> (Output_Get:Vector)
   Event OnEndJumping()
   > RetriggerableDelay(Duration=0.25)
   > Set JumpVelocity(JumpVelocity="0, 0, 0")
   TryGetPawnOwner()
  ]}
]
variables [
 backpack_rotate {type=Rotator}
 backpack_translation {type=Vector}
 JumpVelocity {type=Vector}
]
